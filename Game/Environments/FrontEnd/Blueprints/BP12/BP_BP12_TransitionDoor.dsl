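; Front-end transition door actor, named for BP12, with a battle pass door mesh, a camera and a spinning spotlight.
; On BeginPlay a timeline starts, and each update rotates SpotLight1 locally at a constant rate scaled by world delta time.
; The construction script stores the world transforms of the camera and door mesh in "for Copy" and "For Door" variables.
; Overlap and Tick events are empty.
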(event EventBeginPlay
  (Timeline|Timeline_0 Play
    (:Update
      (Transformation|AddLocalRotation (Variables|Default|GetSpotLight1) (Math|Rotator|ScaleRotator(Float) (Math|Rotator|MakeRotator 0.0 6e+02) (Utilities|Time|GetWorldDeltaSeconds)) false false))))

(event Collision|EventActorBeginOverlap
  (params (OtherActor :type Actor%20Object%20Reference)))

(event EventTick
  (params (DeltaSeconds :type Float%20%28single-precision%29)))

(fn ConstructionScript
  (Variables|TransformforCopy|SetTransform%20for%20Copy (Transformation|GetWorldTransform (Variables|Default|GetCamera)))
  (Variables|TransformforCopy|SetTransform%20For%20Door (Transformation|GetWorldTransform (Variables|Default|GetS_BattlePassDoor_BP12))))
