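; BP_FloorPad is a front-end floor pad that opens or closes according to the item being previewed.
; After a short delay on BeginPlay, it binds to every ItemPreviewSideSwap's OnNewItemSpawned event, whose Should Show Floor flag reverses or plays a timeline.
; The timeline drives the "Open" parameter on two dynamic materials and activates a skydiving effect once at least half open.
; On mobile, construction and BeginPlay move the mesh to the third lighting channel.
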(event EventBeginPlay
  (Utilities|FlowControl|Sequence
    (:then_0
      (if (TouchPlatform|IsMobileGame)
        (Rendering|Components|SetLightingChannels (Variables|Default|GetStaticMeshComponent0) false false true)))
    (:then_1
      (Utilities|FlowControl|Delay 0.2)
      (bind _outactors (Actor|GetAllActorsOfClass "/Game/Environments/FrontEnd/Blueprints/ItemPreview/ItemPreviewSideSwap.ItemPreviewSideSwap_C"))
      (for _array_element _outactors
        (bind _asitem_preview_side_swap (Utilities|Casting|CastToItemPreviewSideSwap _array_element))
        (Default|BindEventtoOnNewItemSpawned _asitem_preview_side_swap (delegate Custom|On%20Item%20Spawned))))))

(event Collision|EventActorBeginOverlap
  (params (OtherActor :type Actor%20Object%20Reference)))

(event Custom|On%20Item%20Spawned
  (params (Should%20Show%20Floor :type Boolean))
  (bind _p_spybase_skydiving_in_place_fx (Variables|Default|GetP_Spybase_Skydiving_In_Place_FX))
  (if Should%20Show%20Floor
    (bind (_direction _openfloor) (Timeline|Timeline_0 Reverse)
      (:Update
        (bind _returnvalue (Rendering|Material|CreateDynamicMaterialInstance (Variables|Default|GetStaticMeshComponent0) 0 0 "None" :!class /Script/Engine.PrimitiveComponent))
        (Rendering|Material|SetScalarParameterValue _returnvalue "Open" _openfloor :!class /Script/Engine.MaterialInstanceDynamic)
        (bind _returnvalue_2 (Rendering|Material|CreateDynamicMaterialInstance (Variables|Default|GetStaticMeshComponent0) 1 0 "None" :!class /Script/Engine.PrimitiveComponent))
        (Rendering|Material|SetScalarParameterValue _returnvalue_2 "Open" _openfloor :!class /Script/Engine.MaterialInstanceDynamic)
        (if (>= _openfloor 0.5)
          (Components|Activation|Activate _p_spybase_skydiving_in_place_fx false)
          (else
            (Components|Activation|Deactivate _p_spybase_skydiving_in_place_fx)))))
    (else
      (Timeline|Timeline_0 Play))))

(event EventTick
  (params (DeltaSeconds :type Float%20%28single-precision%29)))

(fn ConstructionScript
  (if (TouchPlatform|IsMobileGame)
    (Rendering|Components|SetLightingChannels (Variables|Default|GetStaticMeshComponent0) false false true)))
