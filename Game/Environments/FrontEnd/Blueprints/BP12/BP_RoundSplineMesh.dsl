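; BP_RoundSplineMesh is a front-end environment actor that builds a ring, or partial ring, of spline meshes in its construction script.
; It spaces segments evenly by rotating a radius vector around Z, stopping at MaxNumberDrawn.
; Cross-product tangents, scaled by chord length and a tension value, curve each segment along the circle.
; Each segment gets configurable mesh, forward axis, roll, lighting channels and optional material; the event graph is empty.
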(event EventBeginPlay)

(event Collision|EventActorBeginOverlap
  (params (OtherActor :type Actor%20Object%20Reference)))

(event EventTick
  (params (DeltaSeconds :type Float%20%28single-precision%29)))

(fn ConstructionScript
  (bind _roll (Variables|Default|GetRoll))
  (bind _numberofelements (Variables|Default|GetNumberofElements))
  (bind _material (Variables|Default|GetMaterial))
  (for _index (range (+ _numberofelements 1))
    (if (<= _index (Variables|Default|GetMaxNumberDrawn))
      (bind _returnvalue (AddComponent|NODE_AddSplineMeshComponent-0))
      (Rendering|Components|SetLightingChannels _returnvalue (Variables|Default|GetChannel0) (Variables|Default|GetChannel1) (Variables|Default|GetChannel2))
      (SplineMesh|SetForwardAxis _returnvalue (Variables|Default|GetForwardAxis) true)
      (SplineMesh|SetStartandEnd _returnvalue (Math|Vector|RotateVectorAroundAxis (Math|Vector|MakeVector 0.0 (Variables|Default|GetRadius)) (* _index (/ 3.6e+02 (Math|Conversions|ToFloat(Integer) _numberofelements))) "0, 0,1.000000") (* (Math|Vector|Normalize (Math|Vector|CrossProduct (Math|Vector|RotateVectorAroundAxis (Math|Vector|MakeVector 0.0 (Variables|Default|GetRadius)) (* _index (/ 3.6e+02 (Math|Conversions|ToFloat(Integer) _numberofelements))) "0, 0,1.000000") "0, 0,-1.000000")) (* (Math|Vector|Distance(Vector) (Math|Vector|RotateVectorAroundAxis (Math|Vector|MakeVector 0.0 (Variables|Default|GetRadius)) (* _index (/ 3.6e+02 (Math|Conversions|ToFloat(Integer) _numberofelements))) "0, 0,1.000000") (Math|Vector|RotateVectorAroundAxis (Math|Vector|MakeVector 0.0 (Variables|Default|GetRadius)) (* (+ _index 1) (/ 3.6e+02 (Math|Conversions|ToFloat(Integer) _numberofelements))) "0, 0,1.000000")) (Variables|Default|GetTensiom))) (Math|Vector|RotateVectorAroundAxis (Math|Vector|MakeVector 0.0 (Variables|Default|GetRadius)) (* (+ _index 1) (/ 3.6e+02 (Math|Conversions|ToFloat(Integer) _numberofelements))) "0, 0,1.000000") (* (Math|Vector|Normalize (Math|Vector|CrossProduct (Math|Vector|RotateVectorAroundAxis (Math|Vector|MakeVector 0.0 (Variables|Default|GetRadius)) (* (+ _index 1) (/ 3.6e+02 (Math|Conversions|ToFloat(Integer) _numberofelements))) "0, 0,1.000000") "0, 0,-1.000000")) (* (Math|Vector|Distance(Vector) (Math|Vector|RotateVectorAroundAxis (Math|Vector|MakeVector 0.0 (Variables|Default|GetRadius)) (* _index (/ 3.6e+02 (Math|Conversions|ToFloat(Integer) _numberofelements))) "0, 0,1.000000") (Math|Vector|RotateVectorAroundAxis (Math|Vector|MakeVector 0.0 (Variables|Default|GetRadius)) (* (+ _index 1) (/ 3.6e+02 (Math|Conversions|ToFloat(Integer) _numberofelements))) "0, 0,1.000000")) (Variables|Default|GetTensiom))) true :!class /Script/Engine.SplineMeshComponent)
      (SplineMesh|SetStartRollRadians _returnvalue _roll true)
      (SplineMesh|SetEndRollRadians _returnvalue _roll true)
      (Components|StaticMesh|SetStaticMesh _returnvalue (Variables|Default|GetStaticmsh))
      (Utilities|IsValid _material
        (:Is%20Valid
          (Rendering|Material|SetMaterial _returnvalue 0 _material))))))
