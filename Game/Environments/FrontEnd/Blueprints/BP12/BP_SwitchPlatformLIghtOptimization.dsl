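; Front-end environment Blueprint that swaps between PC and Nintendo Switch light sets for platform optimization.
; On BeginPlay it passes the Nintendo Switch platform check to SwitchLights, which shows the Switch array and hides the PC array, or the reverse.
; Each valid entry is cast to SpotLight, falling back to PointLight, to set its light component's visibility.
; The construction script runs SwitchLights with the Debug variable, allowing in-editor preview of either set.
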(event EventBeginPlay
  (CallFunction|SwitchLights (TouchPlatform|IsNintendoSwitch)))

(event Collision|EventActorBeginOverlap
  (params (OtherActor :type Actor%20Object%20Reference)))

(event EventTick
  (params (DeltaSeconds :type Float%20%28single-precision%29)))

(fn SwitchLights
  (params (EnableSwitchPlatform :type Boolean))
  (bind _returnvalue (not EnableSwitchPlatform))
  (Utilities|FlowControl|Sequence
    (:then_0
      (for _array_element (Variables|Default|GetSpotlights_PC)
        (Utilities|IsValid _array_element
          (:Is%20Valid
            (bind _asspot_light (Utilities|Casting|CastToSpotLight _array_element)
              (:then
                (Rendering|SetVisibility (Variables|/Script/Engine.Light|GetLightComponent _asspot_light) _returnvalue false))
              (:CastFailed
                (bind _aspoint_light (Utilities|Casting|CastToPointLight _array_element))
                (Rendering|SetVisibility (Variables|/Script/Engine.Light|GetLightComponent _aspoint_light) _returnvalue false)))))))
    (:then_1
      (for _array_element_2 (Variables|Default|GetSpotlights_Switch)
        (Utilities|IsValid _array_element_2
          (:Is%20Valid
            (bind _asspot_light_2 (Utilities|Casting|CastToSpotLight _array_element_2)
              (:then
                (Rendering|SetVisibility (Variables|/Script/Engine.Light|GetLightComponent _asspot_light_2) EnableSwitchPlatform false))
              (:CastFailed
                (bind _aspoint_light_2 (Utilities|Casting|CastToPointLight _array_element_2))
                (Rendering|SetVisibility (Variables|/Script/Engine.Light|GetLightComponent _aspoint_light_2) EnableSwitchPlatform false)))))))))

(fn ConstructionScript
  (CallFunction|SwitchLights (Variables|Default|GetDebug)))
